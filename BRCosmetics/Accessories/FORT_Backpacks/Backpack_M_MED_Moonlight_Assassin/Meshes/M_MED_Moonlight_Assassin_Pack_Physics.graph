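# Physics asset for the Moonlight Assassin backpack cosmetic, giving collision bodies to the pack and its six dynamic rock bones.
# A tilted capsule covers the pack bone, and each Dyn_rock bone gets a small sphere sized individually.
# Each rock is constrained directly to the pack in a star layout, with swing and twist motion locked.
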PhysicsAsset /BRCosmetics/Accessories/FORT_Backpacks/Backpack_M_MED_Moonlight_Assassin/Meshes/M_MED_Moonlight_Assassin_Pack_Physics
bodies {
 Dyn_rock01 [{shapeType=Sphere, radius=3.6618, center=<-0.1044 0.2709 -0.5369>}]
 Dyn_rock02 [{shapeType=Sphere, radius=3.1822, center=<-0.0669 -0.1167 -0.0076>}]
 Dyn_rock03 [{shapeType=Sphere, radius=1.3519, center=<-0.0598 0.1189 0.0185>}]
 Dyn_rock04 [{shapeType=Sphere, radius=3.1818, center=<0.0397 -0.088 0.0544>}]
 Dyn_rock05 [{shapeType=Sphere, radius=3.4132, center=<-0.5332 0.2487 0.2002>}]
 Dyn_rock06 [{shapeType=Sphere, radius=1.4519, center=<0.0087 -0.0022 0.051>}]
 pack [{shapeType=Capsule, length=6.5, radius=4, center=<0 -7.3724 0>, rotation=<rot 24.9999 0 0>}]
}
constraints [
 {
  bone1Name=Dyn_rock01
  bone2Name=pack
  swing1LimitDegrees=45
  swing1Motion=Locked
  swing2LimitDegrees=45
  swing2Motion=Locked
  twistLimitDegrees=45
  twistMotion=Locked
 }
 {
  bone1Name=Dyn_rock02
  bone2Name=pack
  swing1LimitDegrees=45
  swing1Motion=Locked
  swing2LimitDegrees=45
  swing2Motion=Locked
  twistLimitDegrees=45
  twistMotion=Locked
 }
 {
  bone1Name=Dyn_rock03
  bone2Name=pack
  swing1LimitDegrees=45
  swing1Motion=Locked
  swing2LimitDegrees=45
  swing2Motion=Locked
  twistLimitDegrees=45
  twistMotion=Locked
 }
 {
  bone1Name=Dyn_rock04
  bone2Name=pack
  swing1LimitDegrees=45
  swing1Motion=Locked
  swing2LimitDegrees=45
  swing2Motion=Locked
  twistLimitDegrees=45
  twistMotion=Locked
 }
 {
  bone1Name=Dyn_rock05
  bone2Name=pack
  swing1LimitDegrees=45
  swing1Motion=Locked
  swing2LimitDegrees=45
  swing2Motion=Locked
  twistLimitDegrees=45
  twistMotion=Locked
 }
 {
  bone1Name=Dyn_rock06
  bone2Name=pack
  swing1LimitDegrees=45
  swing1Motion=Locked
  swing2LimitDegrees=45
  swing2Motion=Locked
  twistLimitDegrees=45
  twistMotion=Locked
 }
]
