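# Physics asset for the Moonlight Assassin backpack cosmetic's mesh, giving the pack simple collision shapes for physics.
# It has two capsule bodies: a larger one on the Dyn_spiral bone and a smaller, offset one on the pack bone.
# A single constraint joins Dyn_spiral to pack with swing and twist locked, so it allows no relative rotation between them.
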PhysicsAsset /BRCosmetics/Accessories/FORT_Backpacks/Backpack_M_MED_Moonlight_Assassin/Meshes/M_MED_Moonlight_Assassin_Pack_PhysicsAsset
bodies {
 Dyn_spiral [{shapeType=Capsule, length=12.0043, radius=8.726, center=<0.2264 -0.251 -1.8516>}]
 pack [{shapeType=Capsule, length=3.75, radius=2.75, center=<0 -7.3724 0>}]
}
constraints [{
  bone1Name=Dyn_spiral
  bone2Name=pack
  swing1LimitDegrees=45
  swing1Motion=Locked
  swing2LimitDegrees=45
  swing2Motion=Locked
  twistLimitDegrees=45
  twistMotion=Locked
 }]
