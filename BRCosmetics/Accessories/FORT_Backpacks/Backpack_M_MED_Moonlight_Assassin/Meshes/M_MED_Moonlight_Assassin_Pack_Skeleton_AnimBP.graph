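# Animation Blueprint for the Moonlight Assassin backpack, adding physics motion and floating rock shards to the input pose.
# Two chained RigidBody setups blend by IsRunning, derived from player 0's velocity; the running one adds external force.
# Time-based sine and cosine vectors then additively bob six Dyn_rock bones.
# While targeting, a half-alpha RigidBody pass is layered with the input pose from the pack bone.
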AnimBlueprint /BRCosmetics/Accessories/FORT_Backpacks/Backpack_M_MED_Moonlight_Assassin/Meshes/M_MED_Moonlight_Assassin_Pack_Skeleton_AnimBP
target_skeleton=&/M_MED_Moonlight_Assassin_Pack_Skeleton
graphs [
 AnimGraph {nodes [
   n1 = RigidBody(ComponentPose=Local To Component(LocalPose=Use cached pose 'Transforms'()), Alpha=0.5) {
    bOverrideWorldGravity=true
    overridePhysicsAsset=&/M_MED_Moonlight_Assassin_Pack_Physics
    alphaBoolBlend {blendOption=Linear}
    componentLinearAccScale=<0.75 0.75 0.75>
    componentLinearVelScale=<0.75 0.75 0.75>
    overrideWorldGravity {z=-386}
   }
   n2 = Component To Local(ComponentPose=@n1)
   n3 = Use cached pose 'SubGraph'()
   Output Pose(Result=Blend Poses by bool(
     BlendPose_0=Layered blend per bone(BasePose=@n2, BlendPoses_0=@n3, BlendWeights_0=0.7) {layerSetup [{branchFilters [{blendDepth=1, boneName=pack}]}]}
     BlendPose_1=Use cached pose 'Transforms'()
     bActiveValue=Get bIsTargeting()
    ) {blendType=Linear}) {layerGroup=None}
   AnimGraphNode_SubInput_0 = Input Pose()
   Save cached pose 'SubGraph'(Pose=@AnimGraphNode_SubInput_0) {cacheName=SubGraph}
   n4 = RigidBody(ComponentPose=Local To Component(LocalPose=Use cached pose 'SubGraph'())) {
    bOverrideWorldGravity=true
    overridePhysicsAsset=&M_MED_Moonlight_Assassin_Pack_Physics
    alphaBoolBlend {blendOption=Linear}
    componentLinearAccScale=<0.25 0.25 0.25>
    componentLinearVelScale=<0.25 0.25 0.25>
    overrideWorldGravity {z=-386}
   }
   n5 = RigidBody(ComponentPose=@n4) {
    bOverrideWorldGravity=true
    overridePhysicsAsset=&/M_MED_Moonlight_Assassin_Pack_PhysicsAsset
    alphaBoolBlend {blendOption=Linear}
    overrideWorldGravity {z=-386}
   }
   n6 = RigidBody(ComponentPose=Local To Component(LocalPose=Use cached pose 'SubGraph'()), ExternalForce="0,-10,-25") {
    bOverrideWorldGravity=true
    overridePhysicsAsset=&M_MED_Moonlight_Assassin_Pack_Physics
    alphaBoolBlend {blendOption=Linear}
    overrideWorldGravity {z=-386}
   }
   n7 = RigidBody(ComponentPose=@n6, ExternalForce="0,-10,-25") {
    bOverrideWorldGravity=true
    overridePhysicsAsset=&M_MED_Moonlight_Assassin_Pack_PhysicsAsset
    alphaBoolBlend {blendOption=Linear}
    overrideWorldGravity {z=-386}
   }
   Save cached pose 'RigidBodies'(Pose=Blend Poses by bool(
     BlendPose_0=Component To Local(ComponentPose=@n7)
     BlendPose_1=Component To Local(ComponentPose=@n5)
     BlendTime_0=0.2
     BlendTime_1=0.2
     bActiveValue=Get IsRunning() -> (IsRunning:bool)
    ) {blendType=Linear}) {cacheName=RigidBodies}
   n8 = "Transform (Modify) Bone - Bone: Dyn_rock01"(
    ComponentPose=Local To Component(LocalPose=Use cached pose 'RigidBodies'())
    Alpha=1
    Translation=Get ShardFloatA() -> (ShardFloatA:Vector)
    Rotation="0,0,0"
    Scale="1,1,1"
   ) {translationMode=BMM_Additive, translationSpace=BCS_BoneSpace, alphaBoolBlend {blendOption=Linear}, boneToModify {boneName=Dyn_rock01}}
   n9 = "Transform (Modify) Bone - Bone: Dyn_rock02"(ComponentPose=@n8, Alpha=1, Translation=Get ShardFloatB() -> (ShardFloatB:Vector), Rotation="0,0,0", Scale="1,1,1") {translationMode=BMM_Additive, translationSpace=BCS_BoneSpace, alphaBoolBlend {blendOption=Linear}, boneToModify {boneName=Dyn_rock02}}
   n10 = "Transform (Modify) Bone - Bone: Dyn_rock03"(ComponentPose=@n9, Alpha=1, Translation=Get ShardFloatC() -> (ShardFloatC:Vector), Rotation="0,0,0", Scale="1,1,1") {translationMode=BMM_Additive, translationSpace=BCS_BoneSpace, alphaBoolBlend {blendOption=Linear}, boneToModify {boneName=Dyn_rock03}}
   n11 = Get ShardFloatB()
   n12 = "Transform (Modify) Bone - Bone: Dyn_rock04"(ComponentPose=@n10, Alpha=1, Translation=@n11, Rotation="0,0,0", Scale="1,1,1") {translationMode=BMM_Additive, translationSpace=BCS_BoneSpace, alphaBoolBlend {blendOption=Linear}, boneToModify {boneName=Dyn_rock04}}
   n13 = "Transform (Modify) Bone - Bone: Dyn_rock05"(ComponentPose=@n12, Alpha=1, Translation=@n11, Rotation="0,0,0", Scale="1,1,1") {translationMode=BMM_Additive, translationSpace=BCS_BoneSpace, alphaBoolBlend {blendOption=Linear}, boneToModify {boneName=Dyn_rock05}}
   n14 = "Transform (Modify) Bone - Bone: Dyn_rock06"(ComponentPose=@n13, Alpha=1, Translation=Get ShardFloatC(), Rotation="0,0,0", Scale="1,1,1") {translationMode=BMM_Additive, translationSpace=BCS_BoneSpace, alphaBoolBlend {blendOption=Linear}, boneToModify {boneName=Dyn_rock06}}
   Save cached pose 'Transforms'(Pose=Component To Local(ComponentPose=@n14)) {cacheName=Transforms}
   Use cached pose 'SubGraph'()
   Get ShardFloatA()
   Get ADS() -> (ADS:bool)
  ]}
 EventGraph {nodes [
   Event BlueprintUpdateAnimation()
   > n1 = Sequence()
   n2 = GetPlayerCharacter!(PlayerIndex:int=0) -> (ReturnValue:Character Object Reference)
   > Is Valid(InputObject=@n2)
   n3 = GetVelocity!(self:Actor Object Reference=@n2) -> (ReturnValue:Vector)
   n4 = Get Actor Rotation!(self:Actor Object Reference=@n2) -> (ReturnValue:Rotator) {is=title K2Node_CallFunction}
   n5 = BreakVector(InVec=vector * vector(A:Vector=Get Rotation X Vector!(InRot:Rotator=@n4) -> (ReturnValue:Vector) {is=title K2Node_CallFunction}, B:Vector=@n3) -> (ReturnValue:Vector) {is=title K2Node_CommutativeAssociativeBinaryOperator})
   n6 = float + float(A=@n5, B=@n5.Y) {is=title K2Node_CommutativeAssociativeBinaryOperator}
   n7 = "Not Equal (Float)"!(A:double=BreakVector(InVec=@n3).Z, B:double=0) -> (ReturnValue:bool) {is=title K2Node_CallFunction}
   > n8 = Set pawn_speed_alpha(pawn_speed_alpha:double=float * float(A=MapRangeClamped(Value=@n6, InRangeA=0, InRangeB=500), B=SelectFloat(A=0, B=1, bPickA=@n7))) -> (Output_Get:double)
   > Set IsRunning(IsRunning:bool="float > float"(A=@n8, B=0.4) {is=title K2Node_CallFunction}) -> (Output_Get:bool)
   > Set bUpdateOwnerAnimInputProperty(bUpdateOwnerAnimInputProperty:bool=true) -> (Output_Get:bool)
   > n9 = Branch(Condition=Get bIsTargeting())
   > Set ADS(ADS:bool=true) -> (Output_Get:bool)
   Set ADS(execute=@n9.else, ADS=false)
   n10 = GetGameTimeInSeconds()
   n11 = "Sin (Radians)"(A=float * float(A=float * float(A=1, B=1), B=@n10))
   n12 = "Sin (Radians)"(A=float * float(A=float * float(A=1, B=2), B=@n10))
   n13 = "Sin (Radians)"(A=float * float(A=float * float(A=1, B=1.5), B=@n10))
   Set ShardFloatA(
    execute=@n1.then_1
    ShardFloatA:Vector=MakeVector(X=float * float(A=@n13, B=0.3), Y=float * float(A=@n11, B=0.56), Z=float * float(A=@n12, B=0.22))
   ) -> (Output_Get:Vector)
   n14 = "Cos (Radians)"!(A:double=float * float(A=float * float(A=1, B=1.5), B=@n10)) -> (ReturnValue:double)
   n15 = "Cos (Radians)"(A=float * float(A=float * float(A=1, B=1.78), B=@n10))
   n16 = "Cos (Radians)"(A=float * float(A=float * float(A=1, B=1.15), B=@n10))
   > Set ShardFloatB(ShardFloatB:Vector=MakeVector(X=float * float(A=@n14, B=0.6), Y=float * float(A=@n15, B=0.725), Z=float * float(A=@n16, B=0.3))) -> (Output_Get:Vector)
   n17 = "Sin (Radians)"(A=float * float(A=float * float(A=1, B=1.5), B=@n10))
   n18 = "Sin (Radians)"(A=float * float(A=float * float(A=1, B=2), B=@n10))
   n19 = "Sin (Radians)"(A=float * float(A=float * float(A=1, B=1), B=@n10))
   > Set ShardFloatC(ShardFloatC:Vector=MakeVector(X=float * float(A=@n17, B=0.3), Y=float * float(A=@n19, B=0.6), Z=float * float(A=@n18, B=0.25))) -> (Output_Get:Vector)
   GetPlayerPawn!(PlayerIndex:int=0) -> (ReturnValue:Pawn Object Reference)
   TryGetPawnOwner()
   GetCurrentJointAngles!(execute, self:Skeletal Mesh Component Object Reference=Get Mesh(self=@n2), InBoneName:Name=None) -> (then, Swing1Angle:float, TwistAngle:float, Swing2Angle:float)
   "To String (Float)"(InDouble=0)
   Get bIsCrouched!(self:Character Object Reference=@n2) -> (bIsCrouched:bool)
   Get bIsCrouching()
  ]}
]
variables [
 ADS {type=boolean}
 IsRunning {type=boolean}
 pawn_speed_alpha {type=number}
 ShardFloatA {type=Vector, default=<0.2977 0.4604 0.2059>}
 ShardFloatB {type=Vector, default=<0.0737 -0.165 0.2144>}
 ShardFloatC {type=Vector, default=<0.2977 0.4933 0.234>}
]
